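# BT_Wait is a minimal common behavior tree in AICore that leaves an AI idle.
# It uses the BB_Wait blackboard, with a root Selector holding a single Wait task set to a very long fixed duration.
# The tree has no decorators or services, so it does nothing besides waiting.
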BehaviorTree /AICore/Common/BehaviorTrees/BT/BT_Wait
blackboard=/AICore/Common/BehaviorTrees/BB/BB_Wait
root {
 name=Selector
 class=BTComposite_Selector
 children [Wait {class=BTTask_Wait, properties {waitTime {defaultValue=3600}}}]
 decorators []
 properties {children [{childComposite=null, childTask=&self:BTTask_Wait_0, decorators []}], services []}
 services []
}
